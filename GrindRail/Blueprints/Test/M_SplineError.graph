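# M_SplineError is an unlit, masked material, set up for instanced static meshes, that evidently displays an invalid spline mesh error marker for the GrindRail test Blueprints.
# It samples the InvalidSplineMeshError texture for the marker image.
# The Sprite material function computes a world position offset for a fixed-size, wide camera-facing billboard, and an Offset vector parameter is added to it, lifting the marker along Z by default.
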Material /GrindRail/Blueprints/Test/M_SplineError
graph {
 properties {
  bAutomaticallySetUsageInEditor=false
  bUsedWithInstancedStaticMeshes=true
  blendMode=BLEND_Masked
  shadingModel=MSM_Unlit
  usedShadingModels=Unlit
 }
 nodes [
  Add(
   A=MaterialFunctionCall!(Center Location, 0-1 UVs, XY scale=Constant2Vector() {g=50, r=100}, OverridePivot, "Normals (Optional)") {materialFunction=&Sprite}.World position offset
   B=VectorParameter() {parameterName=Offset, defaultValue=<rgba 0 0 50 1>}.RGB
  )
  TextureSample() {texture=&/InvalidSplineMeshError}
 ]
}
